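# Sound cue for the thrust-stop sound of the PuzzleShed glider.
# A random node picks between two equally weighted wave variants, then a modulator applies slight pitch variation.
# It plays in the gliders sound class with shared glider-thrust attenuation and concurrency settings, at reduced volume.
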SoundCue /BRCosmetics/Sounds/Gliders/PuzzleShed/GliderThrustStop_PuzzleShed_Cue
root=Modulator_0
nodes [
 {id=Modulator_0, children [Random_0], properties {pitchMin=0.99}}
 {id=Random_0, children [WavePlayer_0,WavePlayer_1], properties {weights [1,1]}}
 {id=WavePlayer_0, properties {soundWaveAssetPtr=&/IP_Glider_PuzzleShed_ThrustStop_01}}
 {id=WavePlayer_1, properties {soundWaveAssetPtr=&/IP_Glider_PuzzleShed_ThrustStop_02}}
]
properties {
 attenuationSettings=&Fort_Player_Glider_Thrust_Athena_Att
 maxDistance=12500
 soundClassObject=/Game/Sounds/FortSoundClassesAndModes/Fort_Gliders
 volumeMultiplier=0.5
 concurrencySet [&Athena_Glider_Thrust_Conc]
}
